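# Niagara effect for the Pumpkin Launcher's glowing eye fire, with a User.Variant integer selecting its look.
# EyeFire bursts a persistent mesh particle whose VisibilityTag follows the variant, choosing between two mesh renderers.
# LightBig repeatedly bursts two short-lived, camera-offset glow sprites, using partition index to mirror position and alignment for each eye.
# The variant switches glow color between green and red-pink, while curves grow each sprite and fade its alpha in and out.
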NiagaraSystem /Game/Weapons/FORT_GrenadeLaunchers/Effects/NS_PumpkinLauncher_EyeFire
dependencies {usedDataInterfaces [&NiagaraDataInterfaceCurve], usedDynamicInputs [], usedModules [&SystemState]}
emitters {
 EyeFire {
  emitterUpdateScript [
   &EmitterState(
    Inactive Response=::NewEnumerator1
    Life Cycle Mode=::NewEnumerator1
    Loop Behavior=::NewEnumerator1
    Loop Duration Mode=::NewEnumerator1
    UseLoopDelay=true
   )
   &SpawnBurst_Instantaneous(Use Spawn Probability=false, UseLoopCountLimit=false)
  ]
  particleSpawnScript [
   &InitializeParticle(Mesh Scale Mode=::NewEnumerator4, UsePositionOffset=false, Write Material Random=true)
   &Color(Color=&MakeLinearColorFromVectorAndFloat("Vector (RGB)"=<1 1 1>, "Float (Alpha)"=&ClampFloat(Float=1)))
  ]
  particleUpdateScript [
   &ParticleState(Kill Particles When Lifetime Has Elapsed=false, Let Infinitely Lived Particles Die When Emitter Deactivates=true)
   /SetVariables(Particles.VisibilityTag=$User.Variant) {isSetParameters=true}
  ]
  rendererClasses [&NiagaraMeshRendererProperties]
  renderers [
   {rendererClass=&NiagaraMeshRendererProperties, rendererIndex=0}
   {rendererClass=&NiagaraMeshRendererProperties, rendererIndex=1}
  ]
 }
 LightBig {
  emitterUpdateScript [
   &EmitterState(Inactive Response=::NewEnumerator1, Life Cycle Mode=::NewEnumerator1, Loop Duration=0.2, UseLoopDelay=true)
   &SpawnBurst_Instantaneous(Spawn Count=2, Use Spawn Probability=false, UseLoopCountLimit=false)
  ]
  particleSpawnScript [
   &PartitionParticles(ID to Partition:NiagaraInt32=$Engine.ExecIndex)
   &InitializeParticle(
    Color=&MergeHelper_ColorSwitch(0=<rgba 0.275 2 0.25 0.1>, 1=<rgba 2 0.25 0.3958 0.1>, 2=?, 3=?, Index=$User.Variant, Num=::NewEnumerator0)
    Position=&Lerp_Vector(A=<2 8 -5.5>, B=<-2 8 -5.5>, Alpha=&MakeFloatFromInt(Int=$Particles.PartitionParticles.Partition))
    Position Mode=::NewEnumerator1
    Sprite Size=<5 12>
    Sprite Size Mode=::NewEnumerator4
    UsePositionOffset=false
    Write Material Random=true
   )
   /SetVariables(Particles.SpriteAlignment=&Lerp_Vector(A=<-1 0.3 -0.7>, B=<1 0.3 -0.7>, Alpha=&MakeFloatFromInt(Int=$Particles.PartitionParticles.Partition))) {isSetParameters=true}
  ]
  particleUpdateScript [
   &ParticleState()
   &ScaleSpriteSize(
    Scale Sprite Size Mode=::NewEnumerator0
    Uniform Scale Factor=&FloatFromCurve(FloatCurve=DataInterface() {propertyValues {Curve {keys [<at 0 0.8>,<at 1 1.25>]}}})
   )
   &ScaleColor(
    Scale Alpha=&FloatFromCurve(FloatCurve=DataInterface() {propertyValues {Curve {keys [
         <at 0 0 RCIM_Cubic>
         {arriveTangent=1.1059, interpMode=RCIM_Cubic, leaveTangent=1.1059, time=0.23, value=0.92}
         {arriveTangent=-1.1948, interpMode=RCIM_Cubic, leaveTangent=-1.1948, time=0.8856, value=0.9794}
         <at 1 0 RCIM_Cubic>
        ]}}})
    ScaleA=true
    ScaleRGB=true
   )
   &CameraOffset(Camera Offset Amount=8)
  ]
  rendererClasses [&NiagaraSpriteRendererProperties]
  renderers [{rendererClass=&NiagaraSpriteRendererProperties}]
 }
}
summary {
 systemName=NS_PumpkinLauncher_EyeFire
 emitters [
  {emitterName=EyeFire, rendererClasses [&NiagaraMeshRendererProperties]}
  {emitterName=LightBig, rendererClasses [&NiagaraSpriteRendererProperties]}
 ]
 userVariables [User.Variant {type=NiagaraInt32}]
}
