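# Physics asset for the Egg Launcher Mark II Gadget weapon mesh in the grenade launcher set.
# It has three capsule bodies on the Handle, Front_Handle and Magazine bones.
# Magazine and Front_Handle each constrain to Handle with 45-degree swing and twist limits, forming a simple star rooted at the grip.
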PhysicsAsset /Game/Weapons/FORT_GrenadeLaunchers/Mesh/Egg_Launcher_Mark_II_Gadget_Physics
bodies {
 Front_Handle [{shapeType=Capsule, length=22.8064, radius=18.2722, center=<-0.2135 16.7831 -4.2234>, rotation=<rot 0 1.3567 -73.7791>}]
 Handle [{shapeType=Capsule, length=20.2604, radius=13.6135, center=<-0.9889 -3.4809 9.0352>, rotation=<rot 0 -172.7471 -46.1495>}]
 Magazine [{shapeType=Capsule, length=11.6739, radius=10.6516, center=<-0.1238 5.4933 0.5611>, rotation=<rot 0 3.4396 -54.583>}]
}
constraints [
 {bone1Name=Magazine, bone2Name=Handle, swing1LimitDegrees=45, swing2LimitDegrees=45, twistLimitDegrees=45}
 {bone1Name=Front_Handle, bone2Name=Handle, swing1LimitDegrees=45, swing2LimitDegrees=45, twistLimitDegrees=45}
]
