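# Physics asset for the Chapter 4 Season 4 Shockwave grenade launcher skeletal mesh, defining its simplified collision shapes.
# Two capsule bodies cover the Front_Handle and Handle bones, each offset and rotated from its bone.
# A single constraint joins the two bodies, with matching angular limit values on both swing axes and twist.
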PhysicsAsset /Game/Weapons/FORT_GrenadeLaunchers/Mesh/Grenade_Shockwave_Ch4S4_Launcher_PhysicsAsset
bodies {
 Front_Handle [{shapeType=Capsule, length=20.3296, radius=14.7627, center=<-0.141 18.2948 -11.1265>, rotation=<rot 0 -179.9758 -78.4425>}]
 Handle [{shapeType=Capsule, length=23.3584, radius=13.8492, center=<0.005 1.3252 6.6742>, rotation=<rot 0 -179.9648 -68.0842>}]
}
constraints [{bone1Name=Front_Handle, bone2Name=Handle, swing1LimitDegrees=45, swing2LimitDegrees=45, twistLimitDegrees=45}]
