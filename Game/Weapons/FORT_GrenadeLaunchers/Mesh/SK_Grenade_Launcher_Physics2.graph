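# Physics asset for a Fortnite grenade launcher skeletal mesh, giving it simplified collision for physics simulation.
# Two box bodies cover the Front_Handle and Handle bones.
# A single constraint joins them, with both swing axes and twist set to free rotation.
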PhysicsAsset /Game/Weapons/FORT_GrenadeLaunchers/Mesh/SK_Grenade_Launcher_Physics2
bodies {
 Front_Handle [{shapeType=Box, extentX=21.6525, extentY=41.7529, extentZ=27.7365, center=<-0.0001 16.6062 -10.239>}]
 Handle [{shapeType=Box, extentX=18.6644, extentY=29.9531, extentZ=24.3182, center=<-0.0005 -4.6148 5.6852>}]
}
constraints [{
  bone1Name=Front_Handle
  bone2Name=Handle
  swing1LimitDegrees=45
  swing1Motion=Free
  swing2LimitDegrees=45
  swing2Motion=Free
  twistLimitDegrees=45
  twistMotion=Free
 }]
