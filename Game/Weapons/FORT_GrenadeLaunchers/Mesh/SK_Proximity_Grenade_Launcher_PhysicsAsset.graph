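# Physics asset for the Proximity Grenade Launcher weapon skeletal mesh, giving it a simplified collision shape.
# A single body on the Root bone holds one long, narrow box, offset from the bone origin and yawed so its long axis runs along the bone's Y axis.
# It has no constraints or other bodies, so physics treats the whole launcher as one rigid piece.
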PhysicsAsset /Game/Weapons/FORT_GrenadeLaunchers/Mesh/SK_Proximity_Grenade_Launcher_PhysicsAsset
bodies {Root [{shapeType=Box, extentX=75.5027, extentY=10.9603, extentZ=42.4099, center=<2.1034 22.1239 5.5778>, rotation=<rot 0 -90 0>}]}
constraints []
