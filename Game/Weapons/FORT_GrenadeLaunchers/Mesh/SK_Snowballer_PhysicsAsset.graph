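# Physics asset for the Snowballer skeletal mesh, filed under grenade launcher weapons.
# A single box body on the Handle bone provides the collision shape, longest along Y and offset from the bone origin.
# It has no constraints, so the weapon simulates as one rigid piece with no articulated parts.
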PhysicsAsset /Game/Weapons/FORT_GrenadeLaunchers/Mesh/SK_Snowballer_PhysicsAsset
bodies {Handle [{shapeType=Box, extentX=15.5, extentY=70.5, extentZ=31.5, center=<0 14.4801 8.5522>}]}
constraints []
